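# Sound cue for completing a search of a cemetery coffin loot container, in the loot collect sound class.
# It randomly picks one of two coffin search waves, shapes it with an envelope that fades volume out over about 1.4 seconds, then applies slight random pitch and volume modulation.
# It uses shared prop-search attenuation and a concurrency override that stops the farthest instance before blocking new ones.
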SoundCue /Game/Sounds/Foley_Loot/Containers/Cemetery/Coffin_SearchComplete_Cue
root=Modulator_15
nodes [
 {
  id=Enveloper_21
  children [Random_6]
  properties {pitchCurve {editorCurveData {keys [{time=0}]}}, volumeCurve {editorCurveData {keys [<at 0 1>,<at 1.4041 -0.0146>]}}}
 }
 {id=Modulator_15, children [Enveloper_21], properties {pitchMax=1.07, pitchMin=0.92, volumeMax=1.07, volumeMin=0.92}}
 {id=Random_6, children [WavePlayer_36, WavePlayer_37], properties {weights [1,1]}}
 {id=WavePlayer_36, properties {soundWaveAssetPtr=&/Search_Coffin_01}}
 {id=WavePlayer_37, properties {soundWaveAssetPtr=&/Search_Coffin_02}}
]
properties {
 attenuationSettings=/Game/Sounds/Attenuation/Fort_Prop_Search_Att
 bOverrideConcurrency=true
 maxDistance=1300
 soundClassObject=&Fort_Loot_Collect
 concurrencyOverrides {maxCount=4, resolutionRule=StopFarthestThenPreventNew}
}
