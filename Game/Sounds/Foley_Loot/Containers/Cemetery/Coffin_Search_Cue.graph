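# Sound cue for searching a coffin loot container in the cemetery set, using prop-search attenuation and the loot-collect sound class.
# A quality-level switch has four branches: the first two randomly pick between two coffin search waves, while the last two play only the second wave.
# Each branch passes through a modulator for slight pitch and volume variation.
# A concurrency override caps simultaneous instances, stopping the farthest first.
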SoundCue /Game/Sounds/Foley_Loot/Containers/Cemetery/Coffin_Search_Cue
root=QualityLevel_0
nodes [
 {id=Modulator_0, children [Random_0], properties {pitchMax=1.07, pitchMin=1, volumeMax=1.07, volumeMin=0.92}}
 {id=Modulator_1, children [WavePlayer_3], properties {pitchMax=1.07, pitchMin=1, volumeMax=1.07, volumeMin=0.92}}
 {id=Modulator_15, children [Random_7], properties {pitchMax=1.07, pitchMin=1, volumeMax=1.07, volumeMin=0.92}}
 {id=Modulator_3, children [WavePlayer_4], properties {pitchMax=1.07, pitchMin=1, volumeMax=1.07, volumeMin=0.92}}
 {id=QualityLevel_0, children [Modulator_15,Modulator_0,Modulator_1,Modulator_3]}
 {id=Random_0, children [WavePlayer_0,WavePlayer_1], properties {weights [1,1]}}
 {id=Random_7, children [WavePlayer_40, WavePlayer_41], properties {weights [1,1]}}
 {id=WavePlayer_0, properties {soundWaveAssetPtr=&/Search_Coffin_01}}
 {id=WavePlayer_1, properties {soundWaveAssetPtr=&/Search_Coffin_02}}
 {id=WavePlayer_3, properties {soundWaveAssetPtr=&Search_Coffin_02}}
 {id=WavePlayer_4, properties {soundWaveAssetPtr=&Search_Coffin_02}}
 {id=WavePlayer_40, properties {soundWaveAssetPtr=&Search_Coffin_01}}
 {id=WavePlayer_41, properties {soundWaveAssetPtr=&Search_Coffin_02}}
]
properties {
 attenuationSettings=/Game/Sounds/Attenuation/Fort_Prop_Search_Att
 bOverrideConcurrency=true
 maxDistance=1300
 soundClassObject=&Fort_Loot_Collect
 volumeMultiplier=0.7
 concurrencyOverrides {maxCount=4, resolutionRule=StopFarthestThenPreventNew}
}
